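# Reusable mono one-shot MetaSound source that plays a random wave from a Sounds array, which is empty by default.
# On Play triggers a random get, with its No Repeats count derived from the array length through a Divide node.
# The chosen wave feeds a mono player honoring the source start time, is scaled by Master Volume, and signals On Finished when playback ends.
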MetaSoundSource /GrillMoon_Audio/Assets/MSS_GrillMoon_OneShotArray_Mono
inputs [
 Master Volume {type=Float}
 Sounds {type=WaveAsset:Array}
 UE.Source.OnPlay {type=Trigger}
 UE.Source.StartTime.StartTime {type=Time}
]
interfaces [UE.OutputFormat.Mono:1.0, UE.Source.OneShot:1.0, UE.Source.StartTime:1.0, UE.Source:1.1]
outputs [
 "UE.OutputFormat.Mono.Audio:0" {type=Audio}
 UE.Source.OneShot.OnFinished {type=Trigger}
]
pages [Default {
  input_defaults {Master Volume=1, UE.Source.OnPlay=null, UE.Source.StartTime.StartTime=null, Sounds []}
  nodes [
   ArrayRandomGet_0 = Array.Random Get:WaveAsset:Array(In Array=^Sounds, Next=^UE.Source.OnPlay, No Repeats=UE.Divide:Int32(PrimaryOperand=Array.Num:WaveAsset:Array(Array=^Sounds).Num).Out)
   WavePlayer_0 = UE.Wave Player:Mono(Start Time=^UE.Source.StartTime.StartTime, Play=@ArrayRandomGet_0.On Next, Wave Asset=@ArrayRandomGet_0.Value)
   Multiply_0 = UE.Multiply:Audio by Float(AdditionalOperands=^Master Volume, PrimaryOperand=@WavePlayer_0.Out Mono)
  ]
  links [{dst="@UE.OutputFormat.Mono.Audio:0", src=Multiply_0.Out}, {dst="@UE.Source.OneShot.OnFinished", src=WavePlayer_0.On Finished}]
 }]
properties {assetImportData=&self:AssetImportData}
