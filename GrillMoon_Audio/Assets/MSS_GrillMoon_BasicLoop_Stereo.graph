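# Stereo MetaSound source that endlessly loops a supplied Wave Asset, with a Master Volume input scaling both channels.
# A stereo Wave Player begins on the source's OnPlay trigger at the source start time, then feeds two multiply nodes into the left and right outputs.
# Looping uses a feedback trigger variable: On Nearly Finished is written to it, read back through a deferred accessor, and merged with OnPlay to retrigger playback.
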MetaSoundSource /GrillMoon_Audio/Assets/MSS_GrillMoon_BasicLoop_Stereo
inputs [
 Master Volume {type=Float}
 UE.Source.OnPlay {type=Trigger}
 UE.Source.StartTime.StartTime {type=Time}
 Wave Asset {type=WaveAsset}
]
interfaces [UE.OutputFormat.Stereo:1.0, UE.Source.StartTime:1.0, UE.Source:1.1]
outputs [
 "UE.OutputFormat.Stereo.Audio:0" {type=Audio}
 "UE.OutputFormat.Stereo.Audio:1" {type=Audio}
]
pages [Default {
  input_defaults {Master Volume=1, UE.Source.OnPlay=null, UE.Source.StartTime.StartTime=null, Wave Asset=null}
  nodes [
   n1 = VariableDeferredAccessor.Trigger(Variable=InitVariable.Trigger().Variable)
   TriggerAnyTriggerAny2_0 = "TriggerAny.Trigger Any (2)"(In 0=^UE.Source.OnPlay, In 1=@n1.Value)
   WavePlayer_0 = UE.Wave Player:Stereo(Start Time=^UE.Source.StartTime.StartTime, Wave Asset=^Wave Asset, Play=@TriggerAnyTriggerAny2_0.Out)
   Multiply_0 = UE.Multiply:Audio by Float(AdditionalOperands=^Master Volume, PrimaryOperand=@WavePlayer_0.Out Left)
   Multiply_1 = UE.Multiply:Audio by Float(AdditionalOperands=^Master Volume, PrimaryOperand=@WavePlayer_0.Out Right)
   VariableMutator.Trigger(Variable=@n1.Variable, Value=@WavePlayer_0.On Nearly Finished)
  ]
  links [{dst="@UE.OutputFormat.Stereo.Audio:0", src=Multiply_0.Out}, {dst="@UE.OutputFormat.Stereo.Audio:1", src=Multiply_1.Out}]
 }]
properties {
 assetImportData=&self:AssetImportData
 bLooping=true
 duration=10000
 numChannels=2
 outputFormat=Stereo
 sampleRate=48000
}
